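# Control Rig module for a one-bone IK/FK chain between Start and End connectors, such as a clavicle.
# Construction spawns FK and IK controls, an IK Orient space and, when driving the end bone, switch, Align IK and Stretch channels.
# Forward solve either aims the start bone at IK or follows FK, snapping inactive controls to match.
# Without end-bone driving, IK becomes a proxy and the manipulated control selects the mode.
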ControlRigBlueprint /EpicControlRig/Modules/IkSolves/IkFk1Bone/CRM_IkFk1Bone_v02
graphs [
 ConstructionGraph {nodes [
   RigUnit_PrepareForExecution = &RigUnit_PrepareForExecution()
   /Script/ControlRig.RigUnit_SetupShapeLibraryFromUserData(LibraryName=CRSL, LogShapeLibraries=False, NameSpace=CRSL, Path=ShapeLibrary, ExecutePin=@RigUnit_PrepareForExecution.ExecutePin) {is=struct RigVMUnitNode}
   > RigVMFunction_Sequence = RigVMAggregateNode()
   RigVMVariableNode(
    Value=/Script/ControlRig.RigUnit_ResolveConnector(Connector=(Type=Connector,Name="Start"), SkipSocket=False) {is=struct RigVMUnitNode}.Result
    ExecuteContext=@RigVMFunction_Sequence.A
   ) {variable=Start Bone}
   > RigVMVariableNode(Value=&RigUnit_ResolveConnector(Connector=(Type=Connector,Name="End"), SkipSocket=False).Result) {variable=End Bone}
   n1 = /Script/ControlRig.RigUnit_HierarchyGetParent(bDefaultParent=True, Child=RigVMVariableNode() {variable=Start Bone}.Value) {is=struct RigVMUnitNode}
   n2 = &RigUnit_ResolveConnector(Connector=(Type=Connector,Name="Parent"), SkipSocket=False)
   > RigVMVariableNode(Value=RigVMDispatchNode(False=@n1.Parent, True=@n2.Result, Condition=@n2.bIsConnected) {template="DISPATCH_RigVMDispatch_If(in Condition,in True,in False,out Result)"}.Result) {variable=FK Parent}
   n3 = &RigUnit_ResolveConnector(Connector=(Type=Connector,Name="IK Parent"), SkipSocket=False)
   > RigVMVariableNode(Value=RigVMDispatchNode(False=@n1.Parent, True=@n3.Result, Condition=@n3.bIsConnected) {template="DISPATCH_RigVMDispatch_If(in Condition,in True,in False,out Result)"}.Result) {variable=IK Parent}
   > RigVMVariableNode(Value=&RigUnit_ResolveConnector(Connector=(Type=Connector,Name="IK Orient Space"), SkipSocket=False).Result) {variable=IK Orient Space}
   n4 = &RigVMFunction_ControlFlowBranch(ExecuteContext=@RigVMFunction_Sequence.B, Condition=RigVMVariableNode() {variable=Drive End Bone}.Value)
   RigVMVariableNode(Value=true, ExecuteContext=@n4.False) {variable=IK FK Manipulation}
   > RigVMVariableNode(Value=false) {variable=IK Solve}
   > RigVMVariableNode() {variable=Align IK}
   > RigVMVariableNode() {variable=Stretch}
   RigVMFunction_ControlFlowBranch = &RigVMFunction_ControlFlowBranch(ExecuteContext=@RigVMFunction_Sequence.F, Condition=RigVMVariableNode() {variable=IK FK Manipulation}.Value)
   SetControlDrivenList = /Script/ControlRig.RigUnit_SetControlDrivenList(
    Driven=((Type=None,Name="None"))
    ExecutePin=@RigVMFunction_ControlFlowBranch.True
    Driven.0=RigVMVariableNode() {variable=Start FK Control}.Value
    Control=RigVMVariableNode() {variable=IK Control}.Value.Name
   ) {is=struct RigVMUnitNode, template="SetControlDrivenList::Execute(in Control,in Driven)"}
   > &RigUnit_SetControlDrivenList(
    Driven=((Type=None,Name="None"))
    Control=RigVMVariableNode() {variable=Start FK Control}.Value.Name
    Driven.0=RigVMVariableNode() {variable=IK Control}.Value
   ) {template="SetControlDrivenList::Execute(in Control,in Driven)"}
   n5 = &RigVMFunction_ControlFlowBranch(ExecuteContext=@RigVMFunction_Sequence.G, Condition=RigVMVariableNode() {variable=Drive End Bone}.Value)
   SpawnAnimationChannel = /Script/ControlRig.RigUnit_HierarchyAddAnimationChannelBool(
    LimitsEnabled=()
    MaximumValue=True
    MinimumValue=False
    Name=IK FK Switch
    ExecutePin=@n5.True
    Parent=RigVMVariableNode() {variable=IK Control}.Value
    InitialValue=RigVMVariableNode() {variable=Default IK Interpolation}.Value
   ) {
    is=struct RigVMUnitNode
    template="SpawnAnimationChannel::Execute(in InitialValue,in MinimumValue,in MaximumValue,in LimitsEnabled,in Parent,in Name,out Item)"
   }
   > RigVMVariableNode(Value=@SpawnAnimationChannel.Item) {variable=IK Channel}
   > /Script/ControlRig.RigUnit_SetChannelHosts(
    Hosts=((Type=None,Name="None"),(Type=None,Name="None"))
    Channel=@SpawnAnimationChannel.Item
    Hosts.0=RigVMVariableNode() {variable=Start FK Control}.Value
    Hosts.1=RigVMVariableNode() {variable=End FK Control}.Value
   ) {is=struct RigVMUnitNode}
   > n6 = &RigUnit_HierarchyAddAnimationChannelBool(
    LimitsEnabled=()
    MaximumValue=True
    MinimumValue=False
    Name=Align IK
    Parent=RigVMVariableNode() {variable=IK Control}.Value
    InitialValue=RigVMVariableNode() {variable=Default Align IK}.Value
   ) {template="SpawnAnimationChannel::Execute(in InitialValue,in MinimumValue,in MaximumValue,in LimitsEnabled,in Parent,in Name,out Item)"}
   > RigVMVariableNode(Value=@n6.Item) {variable=Align IK Channel}
   > n7 = &RigUnit_HierarchyAddAnimationChannelBool(
    LimitsEnabled=()
    MaximumValue=True
    MinimumValue=False
    Name=Stretch
    Parent=RigVMVariableNode() {variable=IK Control}.Value
    InitialValue=RigVMVariableNode() {variable=Default Stretch}.Value
   ) {template="SpawnAnimationChannel::Execute(in InitialValue,in MinimumValue,in MaximumValue,in LimitsEnabled,in Parent,in Name,out Item)"}
   > RigVMVariableNode(Value=@n7.Item) {variable=Stretch Channel}
   n8 = RigVMVariableNode() {variable=End Bone}
   Get Control Shape_1 = RigVMFunctionReferenceNode(
    Transform Offset=(Rotation=(X=0,Y=0,Z=0,W=1),Translation=(X=0,Y=0,Z=0),Scale3D=(X=1,Y=1,Z=1))
    Shape=RigVMVariableNode() {variable=Ik Control Shape}.Value
    Transform Offset.Rotation=RigVMVariableNode() {variable=IK Control Rotation Offset}.Value
    Bone Name=@n8.Value.Name
   )
   n9 = &RigUnit_GetTransform(bInitial=False, Item=@n8.Value)
   RigVMFunction_MathBoolNot = &RigVMFunction_MathBoolNot(Value=RigVMVariableNode() {variable=Drive End Bone}.Value)
   n10 = /Script/ControlRig.RigUnit_HierarchyAddControlTransform(
    InitialValue=(Rotation=(X=0,Y=0,Z=0,W=1),Translation=(X=0,Y=0,Z=0),Scale3D=(X=1,Y=1,Z=1))
    Name=IK
    OffsetSpace=GlobalSpace
    OffsetTransform=(Rotation=(X=0,Y=0,Z=0,W=1),Translation=(X=0,Y=0,Z=0),Scale3D=(X=1,Y=1,Z=1))
    Settings="(InitialSpace=LocalSpace,bUsePreferredRotationOrder=False,PreferredRotationOrder=YZX,Limits=(LimitTranslationX=(bMinimum=False,bMaximum=False),LimitTranslationY=(bMinimum=False,bMaximum=False),LimitTranslationZ=(bMinimum=False,bMaximum=False),LimitPitch=(b ...(truncated)"
    Settings.Shape=@"Get Control Shape_1".Result
    OffsetTransform.Scale3D=@n9.Transform.Scale3D
    OffsetTransform.Translation=@n9.Transform.Translation
    OffsetTransform.Rotation=/Script/RigVM.RigVMFunction_MathQuaternionMul(A=@n9.Transform.Rotation, B=RigVMVariableNode() {variable=IK Control Rotation Offset}.Value) {is=struct RigVMUnitNode, template="Multiply::Execute(in A,in B,out Result)"}.Result
    Settings.Proxy.bIsProxy=@RigVMFunction_MathBoolNot.Result
    Parent=RigVMVariableNode() {variable=IK Parent}.Value
    ExecutePin=@RigVMFunction_Sequence.D
   ) {
    is=struct RigVMUnitNode
    template="SpawnControl::Execute(in InitialValue,in Settings,in OffsetTransform,in OffsetSpace,in Parent,in Name,out Item)"
   }
   > RigVMVariableNode(Value=@n10.Item) {variable=IK Control}
   n11 = &RigUnit_GetTransform(bInitial=False, Item=RigVMVariableNode() {variable=Start Bone}.Value)
   n12 = /Script/ControlRig.RigUnit_HierarchyAddBone(
    Name=IK Orient
    Space=GlobalSpace
    Transform=(Rotation=(X=0,Y=0,Z=0,W=1),Translation=(X=0,Y=0,Z=0),Scale3D=(X=1,Y=1,Z=1))
    Transform.Scale3D=@n11.Transform.Scale3D
    Transform.Translation=@n11.Transform.Translation
    Transform.Rotation=&RigUnit_GetTransform(bInitial=False, Item=RigVMVariableNode() {variable=IK Orient Space}.Value).Transform.Rotation
    Parent=RigVMVariableNode() {variable=IK Parent}.Value
    ExecutePin=@RigVMFunction_Sequence.E
   ) {is=struct RigVMUnitNode}
   > RigVMVariableNode(Value=@n12.Item) {variable=IK Orient Space Null}
   > /Script/ControlRig.RigUnit_AddAvailableSpaces(Spaces=((Key=(Type=Control,Name="None"),Label="None")), Control=RigVMVariableNode() {variable=IK Control}.Value, Spaces.0.Key=@n12.Item) {is=struct RigVMUnitNode}
   n13 = RigVMVariableNode() {variable=Start Bone}
   Get Control Shape = RigVMFunctionReferenceNode(
    Transform Offset=(Rotation=(X=0,Y=0,Z=0,W=1),Translation=(X=0,Y=0,Z=0),Scale3D=(X=1,Y=1,Z=1))
    Bone Name=@n13.Value.Name
    Shape=RigVMVariableNode() {variable=FK Control Shape}.Value
   )
   SpawnControl = &RigUnit_HierarchyAddControlTransform(
    InitialValue=(Rotation=(X=0,Y=0,Z=0,W=1),Translation=(X=0,Y=0,Z=0),Scale3D=(X=1,Y=1,Z=1))
    Name=FK
    OffsetSpace=GlobalSpace
    Settings="(InitialSpace=LocalSpace,bUsePreferredRotationOrder=False,PreferredRotationOrder=YZX,Limits=(LimitTranslationX=(bMinimum=False,bMaximum=False),LimitTranslationY=(bMinimum=False,bMaximum=False),LimitTranslationZ=(bMinimum=False,bMaximum=False),LimitPitch=(b ...(truncated)"
    Settings.Shape=@"Get Control Shape".Result
    OffsetTransform=&RigUnit_GetTransform(bInitial=False, Item=@n13.Value).Transform
    Parent=RigVMVariableNode() {variable=FK Parent}.Value
    ExecutePin=@RigVMFunction_Sequence.C
   ) {template="SpawnControl::Execute(in InitialValue,in Settings,in OffsetTransform,in OffsetSpace,in Parent,in Name,out Item)"}
   > RigVMVariableNode(Value=@SpawnControl.Item) {variable=Start FK Control}
   > Set Mirror Axis_1 = RigVMFunctionReferenceNode(Item=@SpawnControl.Item, Value=RigVMVariableNode() {variable=Mirror Axis}.Value)
   n14 = &RigVMFunction_ControlFlowBranch(ExecuteContext=@"Set Mirror Axis_1".ExecuteContext, Condition=RigVMVariableNode() {variable=Drive End Bone}.Value)
   n15 = RigVMVariableNode() {variable=End Bone}
   Get Control Shape_2 = RigVMFunctionReferenceNode(
    Transform Offset=(Rotation=(X=0,Y=0,Z=0,W=1),Translation=(X=0,Y=0,Z=0),Scale3D=(X=1,Y=1,Z=1))
    Shape=RigVMVariableNode() {variable=FK Control Shape}.Value
    Bone Name=@n15.Value.Name
   )
   n16 = &RigUnit_HierarchyAddControlTransform(
    InitialValue=(Rotation=(X=0,Y=0,Z=0,W=1),Translation=(X=0,Y=0,Z=0),Scale3D=(X=1,Y=1,Z=1))
    Name=FK End
    OffsetSpace=GlobalSpace
    Settings="(InitialSpace=LocalSpace,bUsePreferredRotationOrder=False,PreferredRotationOrder=YZX,Limits=(LimitTranslationX=(bMinimum=False,bMaximum=False),LimitTranslationY=(bMinimum=False,bMaximum=False),LimitTranslationZ=(bMinimum=False,bMaximum=False),LimitPitch=(b ...(truncated)"
    Settings.Shape=@"Get Control Shape_2".Result
    OffsetTransform=&RigUnit_GetTransform(bInitial=False, Item=@n15.Value).Transform
    ExecutePin=@n14.True
    Parent=RigVMVariableNode() {variable=Start FK Control}.Value
   ) {template="SpawnControl::Execute(in InitialValue,in Settings,in OffsetTransform,in OffsetSpace,in Parent,in Name,out Item)"}
   > RigVMVariableNode(Value=@n16.Item) {variable=End FK Control}
   > Set Mirror Axis = RigVMFunctionReferenceNode(Value=RigVMVariableNode() {variable=Mirror Axis}.Value, Item=@n16.Item)
  ]}
 New Function_ContainedGraph {nodes [
   RigVMFunctionEntryNode()
   n1 = RigVMVariableNode() {variable=Bone Name}
   n2 = RigVMVariableNode() {variable=Shape}
   RigVMFunction_MathBoolNot = &RigVMFunction_MathBoolNot(Value=RigVMDispatchNode(B=Default, A=@n2.Value.Name) {template="DISPATCH_RigVMDispatch_CoreEquals(in A,in B,out Result)"}.Result)
   n3 = RigVMDispatchNode(
    False=/Script/RigVM.RigVMFunction_NameConcat(B=None, A=&RigVMFunction_NameConcat(A=CRSL., B=@n1.Value) {template="Concat::Execute(in A,in B,out Result)"}.Result) {is=struct RigVMUnitNode, template="Concat::Execute(in A,in B,out Result)"}.Result
    Condition=&RigVMFunction_MathBoolAnd(A=@RigVMFunction_MathBoolNot.Result, B=/Script/ControlRig.RigUnit_ShapeExists(ShapeName=@n2.Value.Name) {is=struct RigVMUnitNode}.Result).Result
    True=@n2.Value.Name
   ) {template="DISPATCH_RigVMDispatch_If(in Condition,in True,in False,out Result)"}
   Get Control Color From Metadata = RigVMFunctionReferenceNode(Color Override=@n2.Value.Color)
   Multiply = &RigVMFunction_MathTransformMul(
    B=/Script/RigVM.RigVMFunction_MathTransformInverse(Value=RigVMVariableNode() {variable=Transform Offset}.Value) {is=struct RigVMUnitNode, template="Inverse::Execute(in Value,out Result)"}.Result
    A=@n2.Value.Transform
   ) {template="Multiply::Execute(in A,in B,out Result)"}
   > RigVMFunctionReturnNode(
    Result=(bVisible=True,Name="Default",Color=(R=1,G=0,B=0,A=1),Transform=(Rotation=(X=0,Y=0,Z=0,W=1),Translation=(X=0,Y=0,Z=0),Scale3D=(X=1,Y=1,Z=1)))
    Result.Name=RigVMDispatchNode(False=Sphere_Solid, True=@n3.Result, Condition=&RigUnit_ShapeExists(ShapeName=@n3.Result).Result) {template="DISPATCH_RigVMDispatch_If(in Condition,in True,in False,out Result)"}.Result
    Result.Color=@"Get Control Color From Metadata".Result
    Result.Transform=@Multiply.Result
    Result.bVisible=@n2.Value.bVisible
   )
  ]}
 New Function_ContainedGraph {nodes [
   n1 = RigVMFunctionEntryNode()
   RigVMFunction_Sequence = &RigVMFunction_Sequence(ExecuteContext=@n1.ExecuteContext)
   n2 = &RigVMFunction_Sequence(ExecuteContext=@RigVMFunction_Sequence.A)
   DISPATCH_RigVMDispatch_ArrayIterator = RigVMDispatchNode(Array=RigVMVariableNode() {variable=IK Controls}.Value, ExecuteContext=@n2.A) {template="DISPATCH_RigVMDispatch_ArrayIterator(in Array,out Element,out Index,out Count,out Ratio)"}
   > SetControlVisibility = /Script/ControlRig.RigUnit_SetControlVisibility(Item=@DISPATCH_RigVMDispatch_ArrayIterator.Element, bVisible=RigVMVariableNode() {variable=IK Solve}.Value) {is=struct RigVMUnitNode, template="SetControlVisibility::Execute(in Item,in Pattern,in bVisible)"}
   n3 = RigVMDispatchNode(Array=RigVMVariableNode() {variable=FK Controls}.Value, ExecuteContext=@n2.B) {template="DISPATCH_RigVMDispatch_ArrayIterator(in Array,out Element,out Index,out Count,out Ratio)"}
   RigVMFunction_MathBoolNot = &RigVMFunction_MathBoolNot(Value=RigVMVariableNode() {variable=IK Solve}.Value)
   > &RigUnit_SetControlVisibility(Item=@n3.Element, bVisible=@RigVMFunction_MathBoolNot.Result) {template="SetControlVisibility::Execute(in Item,in Pattern,in bVisible)"}
   RigVMFunctionReturnNode(ExecuteContext=@RigVMFunction_Sequence.B)
  ]}
 New Function_ContainedGraph {nodes [
   n1 = RigVMFunctionEntryNode()
   RigVMFunction_Sequence = &RigVMFunction_Sequence(ExecuteContext=@n1.ExecuteContext)
   n2 = &RigUnit_SetTransform(
    Weight=1
    bInitial=False
    Value=&RigUnit_GetTransform(bInitial=False, Item=RigVMVariableNode() {variable=Start Bone}.Value).Transform
    ExecutePin=@RigVMFunction_Sequence.A
    Item=RigVMVariableNode() {variable=Start FK Control}.Value
   )
   n3 = &RigVMFunction_ControlFlowBranch(ExecuteContext=@n2.ExecutePin, Condition=RigVMVariableNode() {variable=Drive End Bone}.Value)
   &RigUnit_SetTransform(
    Weight=1
    bInitial=False
    Value=&RigUnit_GetTransform(bInitial=False, Item=RigVMVariableNode() {variable=End Bone}.Value).Transform
    ExecutePin=@n3.True
    Item=RigVMVariableNode() {variable=End FK Control}.Value
   )
   RigVMFunctionReturnNode(ExecuteContext=@RigVMFunction_Sequence.B)
  ]}
 New Function_ContainedGraph {nodes [
   RigVMFunctionEntryNode()
   n1 = &RigUnit_GetTransform(bInitial=False, Item=RigVMVariableNode() {variable=End Bone}.Value)
   > &RigUnit_SetTransform(
    Value=(Rotation=(X=0,Y=0,Z=0,W=1),Translation=(X=0,Y=0,Z=0),Scale3D=(X=1,Y=1,Z=1))
    Weight=1
    bInitial=False
    Value.Scale3D=@n1.Transform.Scale3D
    Value.Translation=@n1.Transform.Translation
    Value.Rotation=&RigVMFunction_MathQuaternionMul(A=@n1.Transform.Rotation, B=RigVMVariableNode() {variable=IK Control Rotation Offset}.Value) {template="Multiply::Execute(in A,in B,out Result)"}.Result
    Item=RigVMVariableNode() {variable=IK Control}.Value
   )
   > RigVMFunctionReturnNode()
  ]}
 RigVMFunctionLibrary {nodes [
   Get Control Shape = RigVMCollapseNode(
    Bone Name=None
    Shape=(bVisible=True,Name="Default",Color=(R=1,G=0,B=0,A=1),Transform=(Rotation=(X=0,Y=0,Z=0,W=1),Translation=(X=0,Y=0,Z=0),Scale3D=(X=1,Y=1,Z=1)))
    Transform Offset=(Rotation=(X=0,Y=0,Z=0,W=1),Translation=(X=0,Y=0,Z=0),Scale3D=(X=1,Y=1,Z=1))
   )
   Hide Show Controls = RigVMCollapseNode(ExecuteContext=(), FK Controls=(), IK Controls=(), IK Solve=false)
   Match FK = RigVMCollapseNode(ExecuteContext=())
   Match IK = RigVMCollapseNode(ExecuteContext=())
  ]}
 RigVMFunction_Sequence_1_ContainedGraph {nodes [
   n1 = RigVMFunctionEntryNode()
   RigVMFunction_Sequence = &RigVMFunction_Sequence(ExecuteContext=@n1.ExecuteContext)
   n2 = &RigVMFunction_Sequence(ExecuteContext=@RigVMFunction_Sequence.B)
   n3 = &RigVMFunction_Sequence(ExecuteContext=@n2.B)
   RigVMFunctionReturnNode(A=@RigVMFunction_Sequence.A, B=@n2.A, C=@n3.A, D=@n3.B)
  ]}
 RigVMFunction_Sequence_1_ContainedGraph {nodes [
   n1 = RigVMFunctionEntryNode()
   RigVMFunction_Sequence = &RigVMFunction_Sequence(ExecuteContext=@n1.ExecuteContext)
   n2 = &RigVMFunction_Sequence(ExecuteContext=@RigVMFunction_Sequence.B)
   n3 = &RigVMFunction_Sequence(ExecuteContext=@n2.B)
   n4 = &RigVMFunction_Sequence(ExecuteContext=@n3.B)
   n5 = &RigVMFunction_Sequence(ExecuteContext=@n4.B)
   n6 = &RigVMFunction_Sequence(ExecuteContext=@n5.B)
   RigVMFunctionReturnNode(A=@RigVMFunction_Sequence.A, B=@n2.A, C=@n3.A, D=@n4.A, E=@n5.A, F=@n6.A, G=@n6.B)
  ]}
 RigVMModel Backward Solve Graph {nodes [
   n1 = /Script/ControlRig.RigUnit_InverseExecution() {is=struct RigVMUnitNode}
   DISPATCH_RigVMDispatch_ArrayMake = RigVMDispatchNode(
    Values=((Type=None,Name="None"),(Type=None,Name="None"))
    Values.0=RigVMVariableNode() {variable=Start Bone}.Value
    Values.1=RigVMVariableNode() {variable=End Bone}.Value
   ) {template="DISPATCH_RigVMDispatch_ArrayMake(in Values,out Array)"}
   DISPATCH_RigVMDispatch_ArrayIterator = RigVMDispatchNode(Array=@DISPATCH_RigVMDispatch_ArrayMake.Array, ExecuteContext=@n1.ExecutePin) {template="DISPATCH_RigVMDispatch_ArrayIterator(in Array,out Element,out Index,out Count,out Ratio)"}
   DISPATCH_RigVMDispatch_ArrayGetAtIndex = RigVMDispatchNode(
    Index=@DISPATCH_RigVMDispatch_ArrayIterator.Index
    Array=RigVMDispatchNode(
     Values=((Type=None,Name="None"),(Type=None,Name="None"))
     Values.1=RigVMVariableNode() {variable=IK Control}.Value
     Values.0=RigVMVariableNode() {variable=Start FK Control}.Value
    ) {template="DISPATCH_RigVMDispatch_ArrayMake(in Values,out Array)"}.Array
   ) {template="DISPATCH_RigVMDispatch_ArrayGetAtIndex(in Array,in Index,out Element)"}
   n2 = &RigUnit_GetTransform(bInitial=False, Item=@DISPATCH_RigVMDispatch_ArrayIterator.Element)
   > &RigUnit_SetTransform(
    Value=(Rotation=(X=0,Y=0,Z=0,W=1),Translation=(X=0,Y=0,Z=0),Scale3D=(X=1,Y=1,Z=1))
    Weight=1
    bInitial=False
    Item=@DISPATCH_RigVMDispatch_ArrayGetAtIndex.Element
    Value.Scale3D=@n2.Transform.Scale3D
    Value.Translation=@n2.Transform.Translation
    Value.Rotation=&RigVMFunction_MathQuaternionMul(A=@n2.Transform.Rotation, B=RigVMVariableNode() {variable=IK Control Rotation Offset}.Value) {template="Multiply::Execute(in A,in B,out Result)"}.Result
   )
  ]}
 RigVMModel Connector Graph {nodes [
   n1 = /Script/ControlRig.RigUnit_ConnectorExecution() {is=struct RigVMUnitNode}
   Item = /Script/ControlRig.RigUnit_Item(Item=(Type=Connector,Name="Start")) {is=struct RigVMUnitNode}
   DISPATCH_RigVMDispatch_ArrayGetAtIndex = RigVMDispatchNode(
    Index=0
    Array=/Script/ControlRig.RigUnit_CollectionChildrenArray(TypeToSearch=All, bDefaultChildren=True, bIncludeParent=False, bRecursive=False, Parent=@Item.Item) {is=struct RigVMUnitNode}.Items
   ) {template="DISPATCH_RigVMDispatch_ArrayGetAtIndex(in Array,in Index,out Element)"}
   Set Default Match To Connector v01_3 = RigVMFunctionReferenceNode(Default Match=(Type=Connector,Name="Clavicle"), ExecuteContext=@n1.ExecuteContext, Connector=@DISPATCH_RigVMDispatch_ArrayGetAtIndex.Element)
   > Set Default Match To Connector v01_4 = RigVMFunctionReferenceNode(
    Connector=(Type=Connector,Name="Parent")
    Default Match=&RigUnit_HierarchyGetParent(bDefaultParent=True, Child=@Item.Item).Parent
   )
   DISPATCH_RigDispatch_GetModuleMetadata = RigVMDispatchNode(Default=(Type=Connector,Name="IK Orient Space"), Name=BodyControl, NameSpace=Root) {template="DISPATCH_RigDispatch_GetModuleMetadata(in Name,in NameSpace,in Default,out Value,out Found)"}
   > Set Default Match To Connector v01_5 = RigVMFunctionReferenceNode(Connector=(Type=Connector,Name="IK Orient Space"), Default Match=@DISPATCH_RigDispatch_GetModuleMetadata.Value)
  ]}
 RigVMModel Forward Solve Event Graph {nodes [
   n1 = /Script/ControlRig.RigUnit_PostBeginExecution() {is=struct RigVMUnitNode}
   n2 = &RigVMFunction_ControlFlowBranch(Condition=/Script/ControlRig.RigUnit_IsInteracting() {is=struct RigVMUnitNode}.bIsInteracting, ExecuteContext=@n1.ExecutePin)
   n3 = RigVMVariableNode() {variable=Start Bone}
   n4 = RigVMVariableNode() {variable=IK Control}
   RigVMVariableNode(
    Value=/Script/RigVM.RigVMFunction_MathQuaternionMakeRelative(
     Parent=&RigUnit_GetTransform(bInitial=false, Item=@n3.Value).Transform.Rotation
     Global=&RigUnit_GetTransform(bInitial=False, Item=@n4.Value).Transform.Rotation
    ) {is=struct RigVMUnitNode, template="Make Relative::Execute(in Global,in Parent,out Local)"}.Local
    ExecuteContext=@n2.False
   ) {variable=IK Control Rotation}
   n5 = &RigVMFunction_ControlFlowBranch(Condition=RigVMVariableNode() {variable=Drive End Bone}.Value, ExecuteContext=@n2.Completed)
   n6 = RigVMVariableNode() {variable=IK Control Rotation Offset}
   n7 = RigVMVariableNode() {variable=IK Control}
   n8 = &RigUnit_GetTransform(bInitial=false, Item=@n7.Value)
   n9 = RigVMVariableNode() {variable=End Bone}
   n10 = /Script/RigVM.RigVMFunction_MathTransformMakeAbsolute(
    Parent=&RigUnit_GetTransform(bInitial=false, Item=@n9.Value).Transform
    Local=&RigVMFunction_MathTransformMakeRelative(
     Global=&RigUnit_GetTransform(bInitial=true, Item=@n7.Value).Transform
     Parent=&RigUnit_GetTransform(bInitial=true, Item=@n9.Value).Transform
    ).Local
   ) {is=struct RigVMUnitNode, template="Make Absolute::Execute(in Local,in Parent,out Global)"}
   /Script/ControlRig.RigUnit_SetShapeTransform(
    Transform=&RigVMFunction_MathTransformMakeAbsolute(
     Local=(Rotation=(X=0,Y=0,Z=0,W=1),Translation=(X=0,Y=0,Z=0),Scale3D=(X=1,Y=1,Z=1))
     Local.Rotation=/Script/RigVM.RigVMFunction_MathQuaternionInverse(Value=@n6.Value) {is=struct RigVMUnitNode, template="Inverse::Execute(in Value,out Result)"}.Result
     Parent=&RigVMFunction_MathTransformMakeRelative(Parent=@n8.Transform, Global=@n10.Global).Local
    ) {template="Make Absolute::Execute(in Local,in Parent,out Global)"}.Global
    ExecutePin=@n5.False
    Control=RigVMVariableNode() {variable=IK Control}.Value.Name
   ) {is=struct RigVMUnitNode}
   RigUnit_BeginExecution = &RigUnit_BeginExecution()
   RigVMFunction_Sequence = RigVMAggregateNode(ExecuteContext=@RigUnit_BeginExecution.ExecutePin)
   n11 = &RigVMFunction_Sequence(ExecuteContext=@RigVMFunction_Sequence.A)
   n12 = &RigUnit_IsInteracting()
   n13 = &RigVMFunction_ControlFlowBranch(
    Condition=&RigVMFunction_MathBoolAnd(B=@n12.bIsInteracting, A=RigVMVariableNode() {variable=IK FK Manipulation}.Value).Result
    ExecuteContext=@n11.B
   )
   n14 = &RigVMFunction_ControlFlowBranch(
    Condition=RigVMDispatchNode(Array=@n12.Items, Element=RigVMVariableNode() {variable=IK Control}.Value) {template="DISPATCH_RigVMDispatch_ArrayFind(in Array,in Element,out Index,out Success)"}.Success
    ExecuteContext=@n13.True
   )
   n15 = &RigVMFunction_ControlFlowBranch(
    Condition=RigVMDispatchNode(
     Array=((Type=None,Name="None"),(Type=None,Name="None"))
     Element=RigVMDispatchNode(Index=0, Array=@n12.Items) {template="DISPATCH_RigVMDispatch_ArrayGetAtIndex(in Array,in Index,out Element)"}.Element
     Array.1=RigVMVariableNode() {variable=End FK Control}.Value
     Array.0=RigVMVariableNode() {variable=Start FK Control}.Value
    ) {template="DISPATCH_RigVMDispatch_ArrayFind(in Array,in Element,out Index,out Success)"}.Success
    ExecuteContext=@n14.False
   )
   RigVMVariableNode(Value=false, ExecuteContext=@n15.True) {variable=IK Solve}
   RigVMVariableNode(Value=true, ExecuteContext=@n14.True) {variable=IK Solve}
   GetAnimationChannelFromItem = /Script/ControlRig.RigUnit_GetBoolAnimationChannelFromItem(bInitial=False, Item=RigVMVariableNode() {variable=IK Channel}.Value) {is=struct RigVMUnitNode, template="GetAnimationChannelFromItem::Execute(out Value,in Item,in bInitial)"}
   RigVMVariableNode(Value=@GetAnimationChannelFromItem.Value, ExecuteContext=@n11.A) {variable=IK Solve}
   > RigVMVariableNode(Value=&RigUnit_GetBoolAnimationChannelFromItem(bInitial=False, Item=RigVMVariableNode() {variable=Align IK Channel}.Value) {template="GetAnimationChannelFromItem::Execute(out Value,in Item,in bInitial)"}.Value) {variable=Align IK}
   > RigVMVariableNode(Value=&RigUnit_GetBoolAnimationChannelFromItem(bInitial=False, Item=RigVMVariableNode() {variable=Stretch Channel}.Value) {template="GetAnimationChannelFromItem::Execute(out Value,in Item,in bInitial)"}.Value) {variable=Stretch}
   n16 = RigVMVariableNode() {variable=IK Parent}
   n17 = RigVMVariableNode() {variable=IK Orient Space Null}
   n18 = /Script/ControlRig.RigUnit_ProjectTransformToNewParent(bChildInitial=True, bNewParentInitial=False, bOldParentInitial=True, NewParent=@n16.Value, OldParent=@n16.Value, Child=@n17.Value) {is=struct RigVMUnitNode}
   Make = RigVMDispatchNode(
    Elements=(Rotation=(X=0,Y=0,Z=0,W=1),Translation=(X=0,Y=0,Z=0),Scale3D=(X=1,Y=1,Z=1))
    Elements.Rotation=&RigUnit_GetTransform(bInitial=False, Item=RigVMVariableNode() {variable=IK Orient Space}.Value).Transform.Rotation
    Elements.Scale3D=@n18.Transform.Scale3D
    Elements.Translation=@n18.Transform.Translation
   ) {template="DISPATCH_RigVMDispatch_MakeStruct(in Elements,out Struct)"}
   > n19 = &RigUnit_SetTransform(Weight=1, bInitial=False, Value=@Make.Struct, Item=@n17.Value)
   n20 = &RigVMFunction_ControlFlowBranch(ExecuteContext=@n19.ExecutePin, Condition=RigVMVariableNode() {variable=IK FK Manipulation}.Value)
   Hide Show Controls = RigVMFunctionReferenceNode(
    FK Controls=((Type=None,Name="None"),(Type=None,Name="None"))
    IK Controls=((Type=None,Name="None"))
    ExecuteContext=@n20.False
    IK Solve=RigVMVariableNode() {variable=IK Solve}.Value
    FK Controls.1=RigVMVariableNode() {variable=End FK Control}.Value
    IK Controls.0=RigVMVariableNode() {variable=IK Control}.Value
    FK Controls.0=RigVMVariableNode() {variable=Start FK Control}.Value
   )
   SetControlVisibility = &RigUnit_SetControlVisibility(bVisible=false, ExecutePin=@n20.True, Item=RigVMVariableNode() {variable=End FK Control}.Value) {template="SetControlVisibility::Execute(in Item,in Pattern,in bVisible)"}
   n21 = &RigUnit_IsInteracting()
   RigVMFunction_MathBoolAnd = &RigVMFunction_MathBoolAnd(A=@n21.bIsInteracting, B=@n21.bIsTranslating)
   RigVMFunction_ControlFlowBranch = &RigVMFunction_ControlFlowBranch(Condition=@RigVMFunction_MathBoolAnd.Result, ExecuteContext=@RigVMFunction_Sequence.B)
   DISPATCH_RigVMDispatch_ArrayGetAtIndex = RigVMDispatchNode(Index=0, Array=@n21.Items) {template="DISPATCH_RigVMDispatch_ArrayGetAtIndex(in Array,in Index,out Element)"}
   n22 = &RigVMFunction_ControlFlowBranch(
    Condition=RigVMDispatchNode(A=@DISPATCH_RigVMDispatch_ArrayGetAtIndex.Element, B=RigVMVariableNode() {variable=IK Control}.Value) {template="DISPATCH_RigVMDispatch_CoreEquals(in A,in B,out Result)"}.Result
    ExecuteContext=@RigVMFunction_ControlFlowBranch.True
   )
   n23 = &RigVMFunction_Sequence(ExecuteContext=@n22.True)
   n24 = &RigVMFunction_ControlFlowBranch(ExecuteContext=@n23.A, Condition=RigVMVariableNode() {variable=Stretch}.Value)
   n25 = &RigUnit_GetTransform(bInitial=False, Item=RigVMVariableNode() {variable=Start Bone}.Value)
   RigVMFunction_MathVectorDistance = &RigVMFunction_MathVectorDistance(
    A=@n25.Transform.Translation
    B=&RigUnit_GetTransform(bInitial=False, Item=RigVMVariableNode() {variable=End Bone}.Value).Transform.Translation
   )
   Subtract = /Script/RigVM.RigVMFunction_MathVectorSub(
    A=&RigUnit_GetTransform(bInitial=False, Item=RigVMVariableNode() {variable=IK Control}.Value).Transform.Translation
    B=@n25.Transform.Translation
   ) {is=struct RigVMUnitNode, template="Subtract::Execute(in A,in B,out Result)"}
   RigVMFunction_MathVectorSetLength = /Script/RigVM.RigVMFunction_MathVectorSetLength(Length=@RigVMFunction_MathVectorDistance.Result, Value=@Subtract.Result) {is=struct RigVMUnitNode, template="SetLength::Execute(in Value,in Length,out Result)"}
   Add = /Script/RigVM.RigVMFunction_MathVectorAdd(B=@RigVMFunction_MathVectorSetLength.Result, A=@n25.Transform.Translation) {is=struct RigVMUnitNode, template="Add::Execute(in A,in B,out Result)"}
   Set Transform_7 = /Script/ControlRig.RigUnit_SetTranslation(
    Space=GlobalSpace
    Weight=1
    bInitial=False
    bPropagateToChildren=True
    Value=@Add.Result
    ExecutePin=@n24.False
    Item=RigVMVariableNode() {variable=IK Control}.Value
   ) {is=struct RigVMUnitNode, template="Set Transform::Execute(in Item,in Space,in bInitial,in Value,in Weight,in bPropagateToChildren)"}
   n26 = &RigVMFunction_ControlFlowBranch(ExecuteContext=@n23.B, Condition=RigVMVariableNode() {variable=Align IK}.Value)
   n27 = &RigUnit_GetTransform(bInitial=false, Item=RigVMVariableNode() {variable=Start Bone}.Value)
   n28 = RigVMVariableNode() {variable=Secondary Axis}
   n29 = /Script/ControlRig.RigUnit_AimBoneMath(
    DebugSettings=(bEnabled=False,Scale=10,WorldOffset=(Rotation=(X=0,Y=0,Z=0,W=1),Translation=(X=0,Y=0,Z=0),Scale3D=(X=1,Y=1,Z=1)))
    Primary=(Weight=1,Axis=(X=1,Y=0,Z=0),Target=(X=1,Y=0,Z=0),Kind=Location,Space=(Type=Bone,Name="None"))
    Secondary=(Weight=1,Axis=(X=0,Y=1,Z=0),Target=(X=1,Y=0,Z=0),Kind=Direction,Space=(Type=Bone,Name="None"))
    Weight=1
    InputTransform=@n27.Transform
    Primary.Target=&RigUnit_GetTransform(bInitial=False, Item=RigVMVariableNode() {variable=IK Control}.Value).Transform.Translation
    Secondary.Target=/Script/RigVM.RigVMFunction_MathQuaternionRotateVector(Transform=@n27.Transform.Rotation, Vector=@n28.Value) {is=struct RigVMUnitNode, template="Rotate Vector::Execute(in Transform,in Vector,out Result)"}.Result
    Secondary.Axis=@n28.Value
    Primary.Axis=RigVMVariableNode() {variable=Primary Axis}.Value
   ) {is=struct RigVMUnitNode}
   Set Transform_8 = &RigUnit_SetRotation(
    Space=GlobalSpace
    Weight=1
    bInitial=False
    Value=/Script/RigVM.RigVMFunction_MathQuaternionMakeAbsolute(Parent=@n29.Result.Rotation, Local=RigVMVariableNode() {variable=IK Control Rotation}.Value) {is=struct RigVMUnitNode, template="Make Absolute::Execute(in Local,in Parent,out Global)"}.Global
    ExecutePin=@n26.True
    Item=RigVMVariableNode() {variable=IK Control}.Value
   )
   n30 = &RigVMFunction_ControlFlowBranch(ExecuteContext=@RigVMFunction_Sequence.C, Condition=RigVMVariableNode() {variable=IK Solve}.Value)
   n31 = &RigUnit_GetTransform(bInitial=false, Item=RigVMVariableNode() {variable=Start Bone}.Value)
   n32 = RigVMVariableNode() {variable=Secondary Axis}
   n33 = &RigUnit_AimBoneMath(
    DebugSettings=(bEnabled=False,Scale=10,WorldOffset=(Rotation=(X=0,Y=0,Z=0,W=1),Translation=(X=0,Y=0,Z=0),Scale3D=(X=1,Y=1,Z=1)))
    Primary=(Weight=1,Axis=(X=1,Y=0,Z=0),Target=(X=1,Y=0,Z=0),Kind=Location,Space=(Type=Bone,Name="None"))
    Secondary=(Weight=1,Axis=(X=0,Y=1,Z=0),Target=(X=1,Y=0,Z=0),Kind=Direction,Space=(Type=Bone,Name="None"))
    Weight=1
    InputTransform=@n31.Transform
    Primary.Target=&RigUnit_GetTransform(bInitial=False, Item=RigVMVariableNode() {variable=IK Control}.Value).Transform.Translation
    Secondary.Target=&RigVMFunction_MathQuaternionRotateVector(Transform=@n31.Transform.Rotation, Vector=@n32.Value) {template="Rotate Vector::Execute(in Transform,in Vector,out Result)"}.Result
    Secondary.Axis=@n32.Value
    Primary.Axis=RigVMVariableNode() {variable=Primary Axis}.Value
   )
   n34 = &RigUnit_SetTransform(Weight=1, bInitial=False, Value=@n33.Result, Item=RigVMVariableNode() {variable=Start Bone}.Value, ExecutePin=@n30.True)
   n35 = &RigVMFunction_ControlFlowBranch(ExecuteContext=@n34.ExecutePin, Condition=RigVMVariableNode() {variable=Drive End Bone}.Value)
   Match FK = RigVMFunctionReferenceNode(ExecuteContext=@n35.Completed)
   n36 = &RigUnit_GetTransform(bInitial=False, Item=RigVMVariableNode() {variable=IK Control}.Value)
   n37 = RigVMVariableNode() {variable=IK Control Rotation Offset}
   &RigUnit_SetTransform(
    Value=(Rotation=(X=0,Y=0,Z=0,W=1),Translation=(X=0,Y=0,Z=0),Scale3D=(X=1,Y=1,Z=1))
    Weight=1
    bInitial=False
    Value.Scale3D=@n36.Transform.Scale3D
    Value.Translation=@n36.Transform.Translation
    Value.Rotation=&RigVMFunction_MathQuaternionMul(
     A=@n36.Transform.Rotation
     B=&RigVMFunction_MathQuaternionInverse(Value=@n37.Value) {template="Inverse::Execute(in Value,out Result)"}.Result
    ) {template="Multiply::Execute(in A,in B,out Result)"}.Result
    ExecutePin=@n35.True
    Item=RigVMVariableNode() {variable=End Bone}.Value
   )
   n38 = &RigUnit_SetTransform(
    Weight=1
    bInitial=False
    Value=&RigUnit_GetTransform(bInitial=False, Item=RigVMVariableNode() {variable=Start FK Control}.Value).Transform
    Item=RigVMVariableNode() {variable=Start Bone}.Value
    ExecutePin=@n30.False
   )
   n39 = &RigVMFunction_ControlFlowBranch(ExecuteContext=@n38.ExecutePin, Condition=RigVMVariableNode() {variable=Drive End Bone}.Value)
   &RigUnit_SetTransform(
    Weight=1
    bInitial=False
    Value=&RigUnit_GetTransform(bInitial=False, Item=RigVMVariableNode() {variable=End FK Control}.Value).Transform
    ExecutePin=@n39.True
    Item=RigVMVariableNode() {variable=End Bone}.Value
   )
   > Match IK = RigVMFunctionReferenceNode()
   Not_Equals = RigVMDispatchNode(B=RigVMVariableNode() {variable=IK Solve}.Value, A=RigVMVariableNode() {variable=Previous IK Solve}.Value) {template="DISPATCH_RigVMDispatch_CoreNotEquals(in A,in B,out Result)"}
   RigVMFunction_MathBoolNot = &RigVMFunction_MathBoolNot(Value=RigVMVariableNode() {variable=IK FK Manipulation}.Value)
   n40 = &RigVMFunction_ControlFlowBranch(
    Condition=&RigVMFunction_MathBoolAnd(B=@Not_Equals.Result, A=@RigVMFunction_MathBoolNot.Result).Result
    ExecuteContext=@RigVMFunction_Sequence.D
   )
   RigVMVariableNode(ExecuteContext=@n40.Completed, Value=RigVMVariableNode() {variable=IK Solve}.Value) {variable=Previous IK Solve}
   For_Each = RigVMDispatchNode(
    Array=((Type=None,Name="None"),(Type=None,Name="None"),(Type=None,Name="None"))
    ExecuteContext=@n40.True
    Array.0=RigVMVariableNode() {variable=IK Control}.Value
    Array.2=RigVMVariableNode() {variable=End FK Control}.Value
    Array.1=RigVMVariableNode() {variable=Start FK Control}.Value
   ) {template="DISPATCH_RigVMDispatch_ArrayIterator(in Array,out Element,out Index,out Count,out Ratio)"}
   > SendEvent = /Script/ControlRig.RigUnit_SendEvent(Event=RequestAutoKey, OffsetInSeconds=0, bEnable=True, bOnlyDuringInteraction=True, Item=@For_Each.Element) {is=struct RigVMUnitNode, template="SendEvent::Execute(in Event,in Item,in OffsetInSeconds,in bEnable,in bOnlyDuringInteraction)"}
   RigVMFunction_UserDefinedEvent = /Script/RigVM.RigVMFunction_UserDefinedEvent(EventName=To IK) {is=struct RigVMUnitNode}
   Match IK_1 = RigVMFunctionReferenceNode(ExecuteContext=@RigVMFunction_UserDefinedEvent.ExecutePin)
   n41 = &RigVMFunction_UserDefinedEvent(EventName=To FK)
   Match FK_1 = RigVMFunctionReferenceNode(ExecuteContext=@n41.ExecutePin)
  ]}
]
hierarchy [
 Start {type=Connector, loc=<0 0 0>, rot=<0 0 0 1>}
 Root {type=Socket, loc=<0 0 0>, rot=<0 0 0 1>}
 Parent {type=Connector, loc=<0 0 0>, rot=<0 0 0 1>}
 IK Orient Space {type=Connector, loc=<0 0 0>, rot=<0 0 0 1>}
 End {type=Connector, loc=<0 0 0>, rot=<0 0 0 1>}
 IK Parent {type=Connector, loc=<0 0 0>, rot=<0 0 0 1>}
]
variables [
 Align IK {type=bool, default=False}
 Align IK Channel {type=FRigElementKey, default=(Type=Bone,Name="")}
 Default Align IK {type=bool, default=False}
 Default IK Interpolation {type=bool, default=True}
 Default Stretch {type=bool, default=True}
 Drive End Bone {type=bool, default=True}
 End Bone {type=FRigElementKey, default=(Type=None,Name="")}
 End FK Control {type=FRigElementKey, default=(Type=Bone,Name="")}
 FK Control Shape {
  type=FRigUnit_HierarchyAddControl_ShapeSettings
  default=(bVisible=True,Name="",Color=(R=1,G=1,B=1,A=1),Transform=(Rotation=(X=0,Y=0,Z=0,W=1),Translation=(X=0,Y=0,Z=0),Scale3D=(X=1,Y=1,Z=1)))
 }
 FK Parent {type=FRigElementKey, default=(Type=None,Name="")}
 IK Channel {type=FRigElementKey, default=(Type=Bone,Name="")}
 IK Control {type=FRigElementKey, default=(Type=Bone,Name="")}
 IK Control Rotation {type=FQuat, default=(X=0,Y=0,Z=0,W=1)}
 IK Control Rotation Offset {type=FQuat, default=(X=0,Y=0,Z=-0.7071,W=0.7071)}
 IK FK Manipulation {type=bool, default=False}
 IK Orient Space {type=FRigElementKey, default=(Type=None,Name="")}
 IK Orient Space Null {type=FRigElementKey, default=(Type=None,Name="")}
 IK Parent {type=FRigElementKey, default=(Type=Bone,Name="")}
 IK Solve {type=bool, default=False}
 Ik Control Shape {
  type=FRigUnit_HierarchyAddControl_ShapeSettings
  default=(bVisible=True,Name="",Color=(R=1,G=1,B=1,A=1),Transform=(Rotation=(X=0,Y=0,Z=0,W=1),Translation=(X=0,Y=0,Z=0),Scale3D=(X=1,Y=1,Z=1)))
 }
 Mirror Axis {type=FVector, default=(X=0,Y=1,Z=1)}
 Previous IK Solve {type=bool, default=False}
 Primary Axis {type=FVector, default=(X=1,Y=0,Z=0)}
 Secondary Axis {type=FVector, default=(X=0,Y=1,Z=0)}
 Start Bone {type=FRigElementKey, default=(Type=None,Name="")}
 Start FK Control {type=FRigElementKey, default=(Type=Bone,Name="")}
 Stretch {type=bool, default=False}
 Stretch Channel {type=FRigElementKey, default=(Type=Bone,Name="")}
]
